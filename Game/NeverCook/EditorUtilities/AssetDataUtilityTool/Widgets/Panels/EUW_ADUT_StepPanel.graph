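# Editor utility panel that lists and adds steps (gatherers, filters, actions, post-gather/post-filter modifiers) on a Fort Asset Data Utility Tool object.
# PanelType selects the tool's array property name and base class; Add opens ADUT_AddMenu with derived classes from the Asset Registry.
# A chosen class is constructed with the tool as outer and appended via Get/SetEditorProperty, then the list rebuilds, as it does on SetObjectToDisplay.
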EditorUtilityWidgetBlueprint /Game/NeverCook/EditorUtilities/AssetDataUtilityTool/Widgets/Panels/EUW_ADUT_StepPanel
graphs [
 AddButtonClicked {nodes [AddButtonClicked() {is=title K2Node_FunctionEntry}]}
 EventGraph {nodes [
   "On Clicked (TA_Generic01_Button)"!() -> (OutputDelegate, then, widget:User Widget Object Reference, event:Pointer Event Structure) {is=title K2Node_ComponentBoundEvent}
   > HandleAddClicked()
   n1 = "On Create Happened (ADUT_AddMenu)"!() -> (OutputDelegate, then, class:Object Class Reference) {is=title K2Node_ComponentBoundEvent}
   > Branch(Condition=IsValidClass!(Class:Object Class Reference=@n1.class) -> (ReturnValue:bool))
   n2 = Get CurrentToolObject() -> (CurrentToolObject:Fort Asset Data Utility Tool Object Reference)
   > Is Valid(InputObject=@n2)
   > PrintString(InString=Append(A="Constructing ", B=Get Class Display Name(Class=@n1.class)))
   > n3 = GetEditorProperty!(execute:Exec, Object:Object Reference=@n2, PropertyName:Name=getPropertyNameFromEnum!() -> (PropertyName:Name=None)) -> (then:Exec, PropertyValue:Array of Object References, ReturnValue:bool)
   > Branch(Condition=@n3.ReturnValue)
   > n4 = Construct Object from Class(Class=@n1.class, self=Get CurrentToolObject()) -> (ReturnValue:Object Reference)
   > Add(TargetArray:Array of Object References=@n3, NewItem:"Object Reference (by ref)"=@n4) {is=title K2Node_CallArrayFunction}
   > SetEditorProperty(Object=Get CurrentToolObject(), PropertyName=getPropertyNameFromEnum() -> (PropertyName=None), PropertyValue:Array of Object References=@n3)
   > UpdatePanelList()
   Event PreConstruct()
   Event Construct()
   Event Tick()
  ]}
 GetChildClasses {nodes [
   GetChildClasses() {is=title K2Node_FunctionEntry}
   n1 = MakeTopLevelAssetPath!(PackageName:String=/Script/FortniteEditor, AssetName:String=FortAssetDataUtilityToolArrayModifier) -> (ReturnValue:Top Level Asset Path Structure)
   n2 = Select!(
    NewEnumerator4:Top Level Asset Path Structure
    NewEnumerator0:Top Level Asset Path Structure=MakeTopLevelAssetPath(PackageName=&FortniteEditor, AssetName=FortAssetDataUtilityToolGatherer)
    NewEnumerator1:Top Level Asset Path Structure=MakeTopLevelAssetPath(PackageName=&FortniteEditor, AssetName=FortAssetDataUtilityToolFilter)
    NewEnumerator2:Top Level Asset Path Structure=MakeTopLevelAssetPath(PackageName=&FortniteEditor, AssetName=FortAssetDataUtilityToolAction)
    NewEnumerator3:Top Level Asset Path Structure=@n1
    NewEnumerator5:Top Level Asset Path Structure=@n1
    Index:ADUT_TypeEnum Enum=Get PanelType() -> (PanelType:ADUT_TypeEnum Enum=NewEnumerator4)
   ) -> (ReturnValue:Top Level Asset Path Structure)
   > n3 = GetDerivedClassNames!(
    execute
    self:Asset Registry Interface=GetAssetRegistry()
    ClassNames:Array of Top Level Asset Path Structures=Make Array("[0]":Top Level Asset Path Structure=@n2) -> (Array:Array of Top Level Asset Path Structures)
    ExcludedClassNames:Set of Top Level Asset Path Structures=Make Set!() -> (Set:Set of Top Level Asset Path Structures) {is=title K2Node_MakeSet}
   ) -> (then, OutDerivedClassNames:Set of Top Level Asset Path Structures)
   > n4 = To Array!(execute, A:Set of Top Level Asset Path Structures=@n3) -> (then, Result:Array of Top Level Asset Path Structures)
   > Remove Item!(execute, TargetArray:Array of Top Level Asset Path Structures=@n4, Item:"Top Level Asset Path Structure (by ref)"=@n2) -> (then, ReturnValue:bool)
   > Return Node!(execute, Result:Array of Top Level Asset Path Structures=@n4)
  ]}
 HandleAddClicked {nodes [
   HandleAddClicked() {is=title K2Node_FunctionEntry}
   > n1 = GetChildClasses!(execute, self) -> (then, Result:Array of Top Level Asset Path Structures) {is=title K2Node_CallFunction}
   > ShowMenu!(
    execute
    self:ADUT Add Menu Object Reference=Get ADUT_AddMenu() -> (ADUT_AddMenu:ADUT Add Menu Object Reference)
    classes:Array of Top Level Asset Path Structures=@n1
   ) -> (then)
  ]}
 SetObjectToDisplay {nodes [
   n1 = SetObjectToDisplay!() -> (then, CurrentToolObject:"Fort Asset Data Utility Tool Object Reference (by ref)") {is=title K2Node_FunctionEntry}
   > Set CurrentToolObject(CurrentToolObject:Fort Asset Data Utility Tool Object Reference=@n1) -> (Output_Get:Fort Asset Data Utility Tool Object Reference)
   > UpdatePanelList()
  ]}
 UpdatePanelList {nodes [
   UpdatePanelList() {is=title K2Node_FunctionEntry}
   > ClearListItems(self=Get CommonListView_38() -> (CommonListView_38:Common List View Object Reference))
   > n1 = GetEditorProperty(Object=Get CurrentToolObject(), PropertyName=getPropertyNameFromEnum() -> (PropertyName=None))
   > Branch(Condition=@n1.ReturnValue)
   > n2 = For Each Loop(Array:Array of Object References=@n1) -> (Array Element:Object Reference)
   > Is Valid(InputObject=@n2)
   > AddItem!(execute, self:List View Object Reference=Get CommonListView_38(), Item:Object Reference=@n2) -> (then)
  ]}
 getPropertyNameFromEnum {nodes [
   getPropertyNameFromEnum() {is=title K2Node_FunctionEntry}
   > Return Node!(
    execute
    PropertyName:Name=Select(
     NewEnumerator4:Name
     NewEnumerator0:Name=Gatherers
     NewEnumerator1:Name=Filters
     NewEnumerator2:Name=Actions
     NewEnumerator3:Name=PostGatherModifiers
     NewEnumerator5:Name=PostFilterModifiers
     Index=Get PanelType() -> (PanelType=NewEnumerator4)
    ) -> (ReturnValue:Name)
   )
  ]}
]
variables [
 AddButtonClicked {default=()}
 CurrentToolObject {type=FortAssetDataUtilityTool, default=null}
 PanelType {type=ADUT_TypeEnum, default=None}
 PropertyName {type=string, default=None}
]
widgets {
 info {rootWidgetClass=&Overlay, widgetCount=6}
 widgets [
  Overlay_1 {widgetClassPath=&Overlay, widget=Overlay_1}
  Border_0 {
   widgetClassPath=&Border
   parent=Overlay_1
   widget=Border_0
   properties {brushColor=<rgba 0.0284 0.0284 0.0284 1>, padding=<ltrb 5 5 5 5>}
   slot {class=OverlaySlot, properties {horizontalAlignment=HAlign_Fill, verticalAlignment=VAlign_Fill}}
  }
  VerticalBox_0 {widgetClassPath=&VerticalBox, parent=Border_0, widget=VerticalBox_0, slot {class=BorderSlot, properties {padding=<ltrb 5 5 5 5>}}}
  TA_Generic01_Button {
   widgetClassPath=&TA_Generic01_Button
   parent=VerticalBox_0
   widget=TA_Generic01_Button
   properties {label=Add, buttonColor_Default=<rgba 0.2295 0.5417 0.0145 1>}
   slot {class=VerticalBoxSlot}
  }
  CommonListView_38 {
   widgetClassPath=&CommonListView
   bIsVariable=true
   parent=VerticalBox_0
   widget=CommonListView_38
   properties {
    entryWidgetClass=/Game/NeverCook/EditorUtilities/AssetDataUtilityTool/Widgets/EUW_PanelListItem.EUW_PanelListItem_C
    horizontalEntrySpacing=3
    numDesignerPreviewEntries=0
    selectionMode=None
    verticalEntrySpacing=3
    scrollBarStyle {
     thickness=8
     draggedThumbImage {
      bIsDynamicallyLoaded=false
      drawAs=RoundedBox
      imageType=NoImage
      mirroring=NoMirror
      resourceName=None
      resourceObject=null
      tiling=NoTile
      imageSize=<0 0>
      margin=<ltrb 0 0 0 0>
      outlineSettings {
       bUseBrushTransparency=false
       roundingType=FixedRadius
       width=0
       color {colorUseRule=UseColor_Specified, specifiedColor=<rgba 0 0 0 0>}
       cornerRadii=<4 4 4 4>
      }
      tintColor {colorUseRule=UseColor_Specified, specifiedColor=<rgba 0.2159 0.2159 0.2159 1>}
      uVRegion {bIsValid=false, max=<0 0>, min=<0 0>}
     }
     horizontalBackgroundImage {
      bIsDynamicallyLoaded=false
      drawAs=NoDrawType
      imageType=NoImage
      mirroring=NoMirror
      resourceName=None
      resourceObject=null
      tiling=NoTile
      imageSize=<0 0>
      margin=<ltrb 0 0 0 0>
      outlineSettings {
       bUseBrushTransparency=false
       roundingType=HalfHeightRadius
       width=0
       color {colorUseRule=UseColor_Specified, specifiedColor=<rgba 0 0 0 0>}
       cornerRadii=<0 0 0 0>
      }
      tintColor {colorUseRule=UseColor_Specified, specifiedColor=<rgba 1 1 1 1>}
      uVRegion {bIsValid=false, max=<0 0>, min=<0 0>}
     }
     horizontalBottomSlotImage {
      bIsDynamicallyLoaded=false
      drawAs=NoDrawType
      imageType=NoImage
      mirroring=NoMirror
      resourceName=None
      resourceObject=null
      tiling=NoTile
      imageSize=<0 0>
      margin=<ltrb 0 0 0 0>
      outlineSettings {
       bUseBrushTransparency=false
       roundingType=HalfHeightRadius
       width=0
       color {colorUseRule=UseColor_Specified, specifiedColor=<rgba 0 0 0 0>}
       cornerRadii=<0 0 0 0>
      }
      tintColor {colorUseRule=UseColor_Specified, specifiedColor=<rgba 1 1 1 1>}
      uVRegion {bIsValid=false, max=<0 0>, min=<0 0>}
     }
     horizontalTopSlotImage {
      bIsDynamicallyLoaded=false
      drawAs=NoDrawType
      imageType=NoImage
      mirroring=NoMirror
      resourceName=None
      resourceObject=null
      tiling=NoTile
      imageSize=<0 0>
      margin=<ltrb 0 0 0 0>
      outlineSettings {
       bUseBrushTransparency=false
       roundingType=HalfHeightRadius
       width=0
       color {colorUseRule=UseColor_Specified, specifiedColor=<rgba 0 0 0 0>}
       cornerRadii=<0 0 0 0>
      }
      tintColor {colorUseRule=UseColor_Specified, specifiedColor=<rgba 1 1 1 1>}
      uVRegion {bIsValid=false, max=<0 0>, min=<0 0>}
     }
     hoveredThumbImage {
      bIsDynamicallyLoaded=false
      drawAs=RoundedBox
      imageType=NoImage
      mirroring=NoMirror
      resourceName=None
      resourceObject=null
      tiling=NoTile
      imageSize=<0 0>
      margin=<ltrb 0 0 0 0>
      outlineSettings {
       bUseBrushTransparency=false
       roundingType=FixedRadius
       width=0
       color {colorUseRule=UseColor_Specified, specifiedColor=<rgba 0 0 0 0>}
       cornerRadii=<4 4 4 4>
      }
      tintColor {colorUseRule=UseColor_Specified, specifiedColor=<rgba 0.2159 0.2159 0.2159 1>}
      uVRegion {bIsValid=false, max=<0 0>, min=<0 0>}
     }
     normalThumbImage {
      bIsDynamicallyLoaded=false
      drawAs=RoundedBox
      imageType=NoImage
      mirroring=NoMirror
      resourceName=None
      resourceObject=null
      tiling=NoTile
      imageSize=<0 0>
      margin=<ltrb 0 0 0 0>
      outlineSettings {
       bUseBrushTransparency=false
       roundingType=FixedRadius
       width=0
       color {colorUseRule=UseColor_Specified, specifiedColor=<rgba 0 0 0 0>}
       cornerRadii=<4 4 4 4>
      }
      tintColor {colorUseRule=UseColor_Specified, specifiedColor=<rgba 0.0953 0.0953 0.0953 1>}
      uVRegion {bIsValid=false, max=<0 0>, min=<0 0>}
     }
     verticalBackgroundImage {
      bIsDynamicallyLoaded=false
      drawAs=NoDrawType
      imageType=NoImage
      mirroring=NoMirror
      resourceName=None
      resourceObject=null
      tiling=NoTile
      imageSize=<0 0>
      margin=<ltrb 0 0 0 0>
      outlineSettings {
       bUseBrushTransparency=false
       roundingType=HalfHeightRadius
       width=0
       color {colorUseRule=UseColor_Specified, specifiedColor=<rgba 0 0 0 0>}
       cornerRadii=<0 0 0 0>
      }
      tintColor {colorUseRule=UseColor_Specified, specifiedColor=<rgba 1 1 1 1>}
      uVRegion {bIsValid=false, max=<0 0>, min=<0 0>}
     }
     verticalBottomSlotImage {
      bIsDynamicallyLoaded=false
      drawAs=NoDrawType
      imageType=NoImage
      mirroring=NoMirror
      resourceName=None
      resourceObject=null
      tiling=NoTile
      imageSize=<0 0>
      margin=<ltrb 0 0 0 0>
      outlineSettings {
       bUseBrushTransparency=false
       roundingType=HalfHeightRadius
       width=0
       color {colorUseRule=UseColor_Specified, specifiedColor=<rgba 0 0 0 0>}
       cornerRadii=<0 0 0 0>
      }
      tintColor {colorUseRule=UseColor_Specified, specifiedColor=<rgba 1 1 1 1>}
      uVRegion {bIsValid=false, max=<0 0>, min=<0 0>}
     }
     verticalTopSlotImage {
      bIsDynamicallyLoaded=false
      drawAs=NoDrawType
      imageType=NoImage
      mirroring=NoMirror
      resourceName=None
      resourceObject=null
      tiling=NoTile
      imageSize=<0 0>
      margin=<ltrb 0 0 0 0>
      outlineSettings {
       bUseBrushTransparency=false
       roundingType=HalfHeightRadius
       width=0
       color {colorUseRule=UseColor_Specified, specifiedColor=<rgba 0 0 0 0>}
       cornerRadii=<0 0 0 0>
      }
      tintColor {colorUseRule=UseColor_Specified, specifiedColor=<rgba 1 1 1 1>}
      uVRegion {bIsValid=false, max=<0 0>, min=<0 0>}
     }
    }
   }
   slot {class=VerticalBoxSlot, properties {size {sizeRule=Fill, value=1}}}
  }
  ADUT_AddMenu {
   widgetClassPath=/Game/NeverCook/EditorUtilities/AssetDataUtilityTool/Widgets/ADUT_AddMenu
   bIsVariable=true
   parent=Overlay_1
   widget=ADUT_AddMenu
   properties {visibility=Collapsed}
   slot {class=OverlaySlot, properties {horizontalAlignment=HAlign_Fill, verticalAlignment=VAlign_Fill}}
  }
 ]
}
