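# Editor utility widget for inspecting a FortAssetDataUtilityTool asset.
# Tool can be set from Content Browser selection or located there; SetToolToDisplay feeds it to five step panels: gatherers, post-gather modifiers, filters, post-filter modifiers, actions.
# Tab selection switches those panels and a parallel set of property-filtered Details Views inside a collapsed border.
# Run Tool button only loads the hard-coded ADUT_Tool_Example into the display; nothing here executes a tool.
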EditorUtilityWidgetBlueprint /Game/NeverCook/EditorUtilities/AssetDataUtilityTool/EUW_AssetDataUtilityInspector
graphs [
 EventGraph {nodes [
   "On Clicked (btnRunTool)"!() -> (OutputDelegate, then, widget:User Widget Object Reference, event:Pointer Event Structure) {is=title K2Node_ComponentBoundEvent}
   > SetToolToDisplay!(execute, self, ToolAsset:Fort Asset Data Utility Tool Object Reference=/Game/Tools/AssetDataUtilityTool/Tools/ADUT_Tool_Example) -> (then)
   n1 = "On Item Selection Changed (listTabs)"!() -> (
    OutputDelegate
    then
    selectedData:Array of Object References
    selectedWidgets:Array of User Widget Object References
    selectedIndices:Array of Integers
   ) {is=title K2Node_ComponentBoundEvent}
   > SetActiveWidgetIndex(
    self:Common Visibility Switcher Object Reference=[
     Get CommonVisibilitySwitcher_183() -> (CommonVisibilitySwitcher_183:Common Visibility Switcher Object Reference)
     Get switcherPanels() -> (switcherPanels:Common Visibility Switcher Object Reference)
    ]
    Index="Get (a copy)"(Array:Array of Integers=@n1.selectedIndices) -> (Output:int)
   )
   "On Clicked (btnFindTool)"!() -> (OutputDelegate, then, widget:User Widget Object Reference, event:Pointer Event Structure) {is=title K2Node_ComponentBoundEvent}
   > n2 = GetPathNameForLoadedAsset(LoadedAsset=Get CurrentTool() -> (CurrentTool:Fort Asset Data Utility Tool Object Reference))
   > SyncBrowserToObjects(AssetPaths=Make Array("[0]":String=@n2) -> (Array:Array of Strings))
   "On Clicked (btnSetTool)"!() -> (OutputDelegate, then, widget:User Widget Object Reference, event:Pointer Event Structure) {is=title K2Node_ComponentBoundEvent}
   > n3 = GetSelectedAssetsOfClass(AssetClass=/Script/FortniteEditor.FortAssetDataUtilityTool)
   > Branch(Condition=Is Valid Index!(TargetArray:Array of Object References=@n3, IndexToTest:int) -> (ReturnValue:bool))
   > n4 = Cast To FortAssetDataUtilityTool!(execute, Object="Get (a copy)"(Array:Array of Object References=@n3) -> (Output:Object Reference)) -> (then, CastFailed, AsFort Asset Data Utility Tool:Fort Asset Data Utility Tool Object Reference) {is=title K2Node_DynamicCast}
   > SetToolToDisplay(ToolAsset=@n4)
   Event Construct()
   > UpdateToolName()
   Event PreConstruct()
   Event Tick()
  ]}
 Init_AddTabs {nodes [
   Init_AddTabs() {is=title K2Node_FunctionEntry}
   > clearItems(self:TA Generic 01 List Object Reference=Get listTabs() -> (listTabs:TA Generic 01 List Object Reference))
   > n1 = Sequence!(execute) -> (then_0, then_1, then_2:Exec, then_3:Exec, then_4:Exec, then_5:Exec)
   > addItemGeneric!(execute, self:TA Generic 01 List Object Reference=Get listTabs(), displayString:String=Gatherers) -> (then, widget:User Widget Object Reference, data:Object Reference)
   addItemGeneric(execute=@n1.then_1, self=Get listTabs(), displayString=Post Gather Modifiers)
   addItemGeneric(execute=@n1.then_2, self=Get listTabs(), displayString=Filters)
   addItemGeneric(execute=@n1.then_3, self=Get listTabs(), displayString=Post Filter Modifiers)
   addItemGeneric(execute=@n1.then_4, self=Get listTabs(), displayString=Actions)
   setIndexSelected!(execute=@n1.then_5, self:TA Generic 01 List Object Reference=Get listTabs(), index:int=0, selected:bool=true, throwEvent:bool=true) -> (then)
  ]}
 SetToolToDisplay {nodes [
   n1 = SetToolToDisplay!() -> (then, ToolAsset:Fort Asset Data Utility Tool Object Reference) {is=title K2Node_FunctionEntry}
   > n2 = Set CurrentTool(CurrentTool:Fort Asset Data Utility Tool Object Reference=@n1) -> (Output_Get:Fort Asset Data Utility Tool Object Reference)
   > n3 = Is Valid(InputObject=@n2)
   SetVisibility(execute=@n3.Is Not Valid, self=Get panelToolSettings() -> (panelToolSettings:Vertical Box Object Reference), InVisibility=Hidden)
   SetObject(
    execute=@n3.Is Valid
    self=[
     Get DetailsView_Filters() -> (DetailsView_Filters:Details View Object Reference)
     Get DetailsView_Gatherers() -> (DetailsView_Gatherers:Details View Object Reference)
     Get DetailsView_Actions() -> (DetailsView_Actions:Details View Object Reference)
     Get DetailsView_FilterMods() -> (DetailsView_FilterMods:Details View Object Reference)
     Get DetailsView_GathererMods() -> (DetailsView_GathererMods:Details View Object Reference)
    ]
    NewObject=Get CurrentTool()
   )
   > Init_AddTabs() {is=title K2Node_CallFunction}
   > UpdateToolName()
   > SetIsEnabled(self=Get btnRunTool() -> (btnRunTool:TA Generic 01 Button Object Reference), bInIsEnabled=true)
   > SetVisibility(self=Get panelToolSettings(), InVisibility=SelfHitTestInvisible)
   > SetObjectToDisplay!(
    execute
    self:EUW ADUT Step Panel Object Reference=[
     Get panelGatherers_Filters() -> (panelGatherers_Filters:EUW ADUT Step Panel Object Reference)
     Get panelGatherers_PostFilter() -> (panelGatherers_PostFilter:EUW ADUT Step Panel Object Reference)
     Get panelGatherers_PostGather() -> (panelGatherers_PostGather:EUW ADUT Step Panel Object Reference)
     Get panelGatherers_Actions() -> (panelGatherers_Actions:EUW ADUT Step Panel Object Reference)
     Get panelGatherers() -> (panelGatherers:EUW ADUT Step Panel Object Reference)
    ]
    CurrentToolObject:"Fort Asset Data Utility Tool Object Reference (by ref)"=Get CurrentTool()
   ) -> (then)
  ]}
 UpdateToolName {nodes [
   UpdateToolName() {is=title K2Node_FunctionEntry}
   > n1 = Is Valid(InputObject=Get CurrentTool())
   "SetText (Text)"(execute=@n1.Is Not Valid, self=Get txtToolName() -> (txtToolName:Common Text Object Reference), InText="INVTEXT(\"No Tool Selected\")")
   n2 = GetPathNameForLoadedAsset(execute=@n1.Is Valid, LoadedAsset=Get CurrentTool())
   n3 = Split(SourceString=@n2, InStr=., SearchDir=FromEnd)
   > "SetText (Text)"(self=Get txtToolName(), InText="To Text (String)"(InString=Append(A="Viewing: ", B=@n3)))
  ]}
]
variables [CurrentTool {type=FortAssetDataUtilityTool, default=null}]
widgets {
 info {rootWidgetClass=&Border, widgetCount=24}
 widgets [
  Border_0 {widgetClassPath=&Border, widget=Border_0, properties {brushColor=<rgba 0.0075 0.0075 0.0075 1>}}
  Overlay_0 {widgetClassPath=&Overlay, parent=Border_0, widget=Overlay_0, slot {class=BorderSlot}}
  VerticalBox_0 {
   widgetClassPath=&VerticalBox
   parent=Overlay_0
   widget=VerticalBox_0
   slot {class=OverlaySlot, properties {horizontalAlignment=HAlign_Fill, verticalAlignment=VAlign_Fill}}
  }
  Border_1 {
   widgetClassPath=&Border
   parent=VerticalBox_0
   widget=Border_1
   properties {brushColor=<rgba 1 1 1 0.1>, padding=<ltrb 0 0 0 0>}
   slot {class=VerticalBoxSlot}
  }
  HorizontalBox_1 {
   widgetClassPath=&HorizontalBox
   parent=Border_1
   widget=HorizontalBox_1
   slot {class=BorderSlot, properties {padding=<ltrb 0 0 0 0>}}
  }
  txtToolName {
   widgetClassPath=&CommonTextBlock
   bIsVariable=true
   parent=HorizontalBox_1
   widget=txtToolName
   properties {
    font {
     bForceMonospaced=false
     bMaterialIsStencil=false
     fontMaterial=null
     fontObject=&NotoSans
     letterSpacing=0
     monospacedWidth=1
     size=12
     skewAmount=0
     typefaceFontName=NotoSans-Regular
     outlineSettings {
      bApplyOutlineToDropShadows=false
      bMiteredCorners=false
      bSeparateFillAlpha=false
      outlineMaterial=null
      outlineSize=0
      outlineColor=<rgba 0 0 0 1>
     }
    }
    shadowOffset=<0 0>
   }
   slot {class=HorizontalBoxSlot, properties {padding=<ltrb 3 3 3 3>, size {sizeRule=Fill, value=1}}}
  }
  btnSetTool {
   widgetClassPath=&TA_Generic01_Button
   parent=HorizontalBox_1
   widget=btnSetTool
   properties {iconBrush {
     bIsDynamicallyLoaded=false
     drawAs=Image
     imageType=NoImage
     mirroring=NoMirror
     resourceName=None
     resourceObject=/Game/Athena/UI/InGame/Creative/PropertyEditor/Icon_Picker/Textures_V2/64x64/T_UI_IconLibrary_LeftArrow_64
     tiling=NoTile
     imageSize=<20 20>
     margin=<ltrb 0 0 0 0>
     outlineSettings {
      bUseBrushTransparency=false
      roundingType=HalfHeightRadius
      width=0
      color {colorUseRule=UseColor_Specified, specifiedColor=<rgba 0 0 0 0>}
      cornerRadii=<0 0 0 1>
     }
     tintColor {colorUseRule=UseColor_Specified, specifiedColor=<rgba 1 1 1 1>}
     uVRegion {bIsValid=false, max=<0 0>, min=<0 0>}
    }}
   slot {class=HorizontalBoxSlot, properties {verticalAlignment=VAlign_Center, padding=<ltrb 0 0 5 0>}}
  }
  btnFindTool {
   widgetClassPath=&TA_Generic01_Button
   parent=HorizontalBox_1
   widget=btnFindTool
   properties {iconBrush {
     bIsDynamicallyLoaded=false
     drawAs=Image
     imageType=NoImage
     mirroring=NoMirror
     resourceName=None
     resourceObject=/Game/Athena/UI/Challenges/Art/ChallengeTypeIcons/UI_ChallengeIcon_Find
     tiling=NoTile
     imageSize=<20 20>
     margin=<ltrb 0 0 0 0>
     outlineSettings {
      bUseBrushTransparency=false
      roundingType=HalfHeightRadius
      width=0
      color {colorUseRule=UseColor_Specified, specifiedColor=<rgba 0 0 0 0>}
      cornerRadii=<0 0 0 1>
     }
     tintColor {colorUseRule=UseColor_Specified, specifiedColor=<rgba 1 1 1 1>}
     uVRegion {bIsValid=false, max=<0 0>, min=<0 0>}
    }}
   slot {class=HorizontalBoxSlot, properties {verticalAlignment=VAlign_Center}}
  }
  panelToolSettings {
   widgetClassPath=&VerticalBox
   bIsVariable=true
   parent=VerticalBox_0
   widget=panelToolSettings
   properties {visibility=Hidden}
   slot {class=VerticalBoxSlot, properties {padding=<ltrb 0 10 0 0>, size {sizeRule=Fill, value=1}}}
  }
  listTabs {
   widgetClassPath=/Game/Blueprints/Blutilities/ForTechArt/Widgets/List/TA_Generic01_List
   bIsVariable=true
   parent=panelToolSettings
   widget=listTabs
   properties {
    allowMultiSelect=false
    itemPadding=5
    listItemClass=/Game/Blueprints/Blutilities/ForTechArt/Widgets/Buttons/TA_Generic01_TabItem.TA_Generic01_TabItem_C
    scrollOrientation=Orient_Horizontal
   }
   slot {class=VerticalBoxSlot}
  }
  switcherPanels {
   widgetClassPath=&CommonVisibilitySwitcher
   bIsVariable=true
   parent=panelToolSettings
   widget=switcherPanels
   slot {class=VerticalBoxSlot, properties {size {sizeRule=Fill, value=1}}}
  }
  panelGatherers {
   widgetClassPath=/Game/NeverCook/EditorUtilities/AssetDataUtilityTool/Widgets/Panels/EUW_ADUT_StepPanel
   bIsVariable=true
   parent=switcherPanels
   widget=panelGatherers
   properties {panelType=Gatherer, propertyName=Gatherers}
   slot {class=CommonVisibilitySwitcherSlot}
  }
  panelGatherers_PostGather {
   widgetClassPath=&EUW_ADUT_StepPanel
   bIsVariable=true
   parent=switcherPanels
   widget=panelGatherers_PostGather
   properties {panelType=Post Gather Modifier, propertyName=PostGatherModifiers}
   slot {class=CommonVisibilitySwitcherSlot}
  }
  panelGatherers_Filters {
   widgetClassPath=&EUW_ADUT_StepPanel
   bIsVariable=true
   parent=switcherPanels
   widget=panelGatherers_Filters
   properties {panelType=Filter, propertyName=Filters}
   slot {class=CommonVisibilitySwitcherSlot}
  }
  panelGatherers_PostFilter {
   widgetClassPath=&EUW_ADUT_StepPanel
   bIsVariable=true
   parent=switcherPanels
   widget=panelGatherers_PostFilter
   properties {panelType=Post Filter Modifier, propertyName=PostFilterModifiers}
   slot {class=CommonVisibilitySwitcherSlot}
  }
  panelGatherers_Actions {
   widgetClassPath=&EUW_ADUT_StepPanel
   bIsVariable=true
   parent=switcherPanels
   widget=panelGatherers_Actions
   properties {panelType=Action, propertyName=Actions}
   slot {class=CommonVisibilitySwitcherSlot}
  }
  Border_2 {
   widgetClassPath=&Border
   parent=panelToolSettings
   widget=Border_2
   properties {visibility=Collapsed, brushColor=<rgba 1 1 1 0.05>}
   slot {class=VerticalBoxSlot, properties {size {sizeRule=Fill, value=1}}}
  }
  CommonVisibilitySwitcher_183 {
   widgetClassPath=&CommonVisibilitySwitcher
   bIsVariable=true
   parent=Border_2
   widget=CommonVisibilitySwitcher_183
   slot {class=BorderSlot}
  }
  DetailsView_Gatherers {
   widgetClassPath=&DetailsView
   parent=CommonVisibilitySwitcher_183
   widget=DetailsView_Gatherers
   properties {propertiesToShow [Gatherers]}
   slot {class=CommonVisibilitySwitcherSlot}
  }
  DetailsView_GathererMods {
   widgetClassPath=&DetailsView
   parent=CommonVisibilitySwitcher_183
   widget=DetailsView_GathererMods
   properties {propertiesToShow [PostGatherModifiers]}
   slot {class=CommonVisibilitySwitcherSlot}
  }
  DetailsView_Filters {
   widgetClassPath=&DetailsView
   parent=CommonVisibilitySwitcher_183
   widget=DetailsView_Filters
   properties {propertiesToShow [Filters]}
   slot {class=CommonVisibilitySwitcherSlot}
  }
  DetailsView_FilterMods {
   widgetClassPath=&DetailsView
   parent=CommonVisibilitySwitcher_183
   widget=DetailsView_FilterMods
   properties {propertiesToShow [PostFilterModifiers]}
   slot {class=CommonVisibilitySwitcherSlot}
  }
  DetailsView_Actions {
   widgetClassPath=&DetailsView
   parent=CommonVisibilitySwitcher_183
   widget=DetailsView_Actions
   properties {propertiesToShow [Actions]}
   slot {class=CommonVisibilitySwitcherSlot}
  }
  btnRunTool {
   widgetClassPath=&TA_Generic01_Button
   parent=panelToolSettings
   widget=btnRunTool
   properties {bIsEnabled=false, label=Run Tool}
   slot {class=VerticalBoxSlot, properties {padding=<ltrb 0 10 0 0>, size {sizeRule=Automatic, value=0.25}}}
  }
 ]
}
